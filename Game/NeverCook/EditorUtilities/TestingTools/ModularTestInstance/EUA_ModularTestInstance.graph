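# Editor utility actor that spawns a transient BuildingLevelInstance of a chosen map at its own transform for modular testing, remembering that choice.
# Construction script debounces through a short timer; DoTheThing loads or saves the choice, updates or destroys the instance, and toggles the Cube placeholder.
# Choices persist in an MTI_SaveData child Blueprint's default-object map, keyed by level and actor label, using Python to fetch the CDO.
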EditorUtilityBlueprint /Game/NeverCook/EditorUtilities/TestingTools/ModularTestInstance/EUA_ModularTestInstance
graphs [
 CreateSaveAsset {nodes [
   CreateSaveAsset() {is=title K2Node_FunctionEntry}
   > n1 = DoesAssetExist(AssetPath=GetPathInfoForSave!() -> (path:String, name:String, package_name:String, class_name:String).package_name)
   > n2 = Branch(Condition=@n1)
   n3 = Execute Python Script!(
    execute=[@n13.then, @n2.then]
    PythonScript="import unreal\r\nmy_class = unreal.load_class(None, in_class_name)\r\ncdo = unreal.get_default_object(my_class)"
    in_class_name:String=GetPathInfoForSave().class_name
   ) -> (then, ReturnValue, cdo:Object Reference) {inputs [in_class_name], outputs [cdo]}
   > n4 = Set l_cdo(l_cdo:Object Reference=@n3.cdo) -> (Output_Get:Object Reference)
   > n5 = GetEditorProperty!(execute:Exec, Object:Object Reference=@n4, PropertyName:Name=Get PropertyName_SaveDataMap() -> (PropertyName_SaveDataMap:Name=None)) -> (then:Exec, PropertyValue:Map of Names to Soft World Reference Structures, ReturnValue:bool)
   > Set l_map(l_map:Map of Names to Soft World Reference Structures=@n5) -> (Output_Get:Map of Names to Soft World Reference Structures)
   > n6 = Sequence!(execute) -> (then_0, then_1, then_2:Exec)
   n7 = Find(
    TargetMap:Map of Names to Soft World Reference Structures=Get l_map() -> (l_map:Map of Names to Soft World Reference Structures)
    Key:"Name (by ref)"=GetActorIDForSave!() -> (ReturnValue:Name=None)
   ) -> (Value:Soft World Reference Structure)
   > Branch(Condition=@n7.ReturnValue)
   n8 = Get privateMapToLoad() -> (privateMapToLoad:Soft World Reference Structure)
   > Branch(Condition="Equal (SoftObjectReference)"(
     A:"Soft Object Reference (by ref)"=Break Soft World Reference!(SoftWorldReference:"Soft World Reference Structure (by ref)"=@n7) -> (WorldAsset:World Soft Object Reference) {is=title K2Node_BreakStruct}
     B:"Soft Object Reference (by ref)"=Break Soft World Reference(SoftWorldReference=@n8)
    ) -> (ReturnValue:bool=~) {is=title K2Node_PromotableOperator})
   > Return Node()
   Add!(
    execute=@n6.then_1
    TargetMap:Map of Names to Soft World Reference Structures=Get l_map()
    Key:"Name (by ref)"=GetActorIDForSave() -> (ReturnValue=None)
    Value:"Soft World Reference Structure (by ref)"=Get privateMapToLoad()
   ) -> (then) {is=title K2Node_CallFunction}
   > SetEditorProperty(
    Object=Get l_cdo() -> (l_cdo:Object Reference)
    PropertyName=Get PropertyName_SaveDataMap() -> (PropertyName_SaveDataMap=None)
    PropertyValue:Map of Names to Soft World Reference Structures=Get l_map()
   )
   SaveAsset!(execute=@n6.then_2, AssetToSave:String=GetPathInfoForSave().package_name, bOnlyIfIsDirty:bool=true) -> (then, ReturnValue:bool)
   n9 = GetAssetTools(execute=@n2.else)
   n10 = GetPathInfoForSave()
   > n11 = CreateAsset(self=@n9, AssetName=@n10.name, PackagePath=@n10, AssetClass=/Script/Engine.Blueprint)
   > n12 = GetBlueprintAsset(Object=@n11)
   > n13 = ReparentBlueprint(Blueprint=@n12, NewParentClass=/Game/NeverCook/EditorUtilities/TestingTools/ModularTestInstance/Data/MTI_SaveData.MTI_SaveData_C)
  ]}
 DoTheThing {nodes [
   DoTheThing() {is=title K2Node_FunctionEntry}
   > n1 = Sequence()
   > n2 = Branch(Condition=Get hasLoaded() -> (hasLoaded:bool))
   n3 = Get privateMapToLoad()
   n4 = Break Soft World Reference(SoftWorldReference=Get MapToLoad() -> (MapToLoad:Soft World Reference Structure))
   > Branch(Condition="Not Equal (SoftObjectReference)"(A:"Soft Object Reference (by ref)"=@n4, B:"Soft Object Reference (by ref)"=Break Soft World Reference(SoftWorldReference=@n3)) -> (ReturnValue:bool=~) {is=title K2Node_PromotableOperator})
   > SetEditorProperty(Object=Self-Reference(), PropertyName=privateMapToLoad, PropertyValue:Soft World Reference Structure=Get MapToLoad())
   > Branch(Condition=IsValidSoftObjectReference(SoftObjectReference=@n4))
   > CreateSaveAsset() {is=title K2Node_CallFunction}
   LoadData(execute=@n2.else)
   n5 = Get privateMapToLoad()
   n6 = Branch(execute=@n1.then_1, Condition=IsValidSoftObjectReference(SoftObjectReference=Break Soft World Reference(SoftWorldReference=@n5)))
   n7 = Resolve Soft Reference!(Input:Building Level Instance Soft Object Reference=Get currentSpawnedActorSoftRef() -> (currentSpawnedActorSoftRef:Building Level Instance Soft Object Reference)) -> (Output:Building Level Instance Object Reference)
   > n8 = Is Valid(InputObject=@n7)
   n9 = BreakTransform(InTransform=Get Actor Transform())
   n10 = SpawnActorFromClass!(
    execute=@n8.Is Not Valid
    self:Editor Actor Subsystem Object Reference=Get EditorActorSubsystem()
    ActorClass:Actor Class Reference=/Script/FortniteGame.BuildingLevelInstance
    Location:Vector=@n9
    Rotation:Rotator=@n9.Rotation
    bTransient:bool=true
   ) -> (then, ReturnValue:Building Level Instance Object Reference)
   > SetEditorProperty(
    Object=Self-Reference()
    PropertyName=currentSpawnedActorSoftRef
    PropertyValue:Soft Object Reference=To Soft Object Reference(SoftObjectPath=GetSoftObjectPath(Object=@n10))
    ChangeNotifyMode=Never
   )
   > n11 = SetActorLabel(self=@n10, NewActorLabel=Append(A=GetActorLabel(bCreateIfNone=true), B=_SpawnedLevel), bMarkDirty=true)
   SetEditorProperty(
    execute=[@n11.then, @n13.then]
    Object=Resolve Soft Reference(Input=Get currentSpawnedActorSoftRef())
    PropertyName=WorldAsset
    PropertyValue:World Soft Object Reference=Break Soft World Reference(SoftWorldReference=Get privateMapToLoad())
   )
   n12 = BreakTransform(InTransform=Get Actor Transform())
   n13 = SetActorTransform!(
    execute=@n8.Is Valid
    self:Editor Actor Subsystem Object Reference=Get EditorActorSubsystem()
    InActor:Actor Object Reference=@n7
    InWorldTransform:"Transform (by ref)"=MakeTransform(Location=@n12, Rotation=@n12.Rotation)
   ) -> (then, ReturnValue:bool)
   n14 = Resolve Soft Reference(Input=Get currentSpawnedActorSoftRef())
   Is Valid(exec=@n6.else, InputObject=@n14)
   > DestroyActor!(execute, self:Editor Actor Subsystem Object Reference=Get EditorActorSubsystem(), ActorToDestroy:Actor Object Reference=@n14) -> (then, ReturnValue:bool)
   > n15 = Set currentSpawnedActorSoftRef(currentSpawnedActorSoftRef:Building Level Instance Soft Object Reference) -> (Output_Get:Building Level Instance Soft Object Reference)
   > SetEditorProperty(
    Object=Self-Reference()
    PropertyName=currentSpawnedActorSoftRef
    PropertyValue:Building Level Instance Soft Object Reference=@n15
    ChangeNotifyMode=Never
   )
   n16 = Get currentSpawnedActorSoftRef()
   SetVisibility!(
    execute=@n1.then_2
    self:Scene Component Object Reference=Get Cube() -> (Cube:Static Mesh Component Object Reference)
    bNewVisibility:bool=NOT Boolean(A=IsValidSoftObjectReference(SoftObjectReference=@n16))
    bPropagateToChildren:bool=false
   ) -> (then)
  ]}
 EventGraph {links [], nodes []}
 GetActorIDForSave {nodes [
   GetActorIDForSave() {is=title K2Node_FunctionEntry}
   > n1 = GetCurrentLevelName!(execute, WorldContextObject:Object Reference, bRemovePrefixString:bool=true) -> (then, ReturnValue:String)
   n2 = GetActorLabel(bCreateIfNone=true)
   > Return Node!(execute, ReturnValue:Name=String To Name(InString=Append!(A=@n1, B=:, C:String=@n2) -> (ReturnValue)))
  ]}
 GetPathInfoForSave {nodes [
   GetPathInfoForSave() {is=title K2Node_FunctionEntry}
   n1 = Get SaveAssetName() -> (SaveAssetName:String)
   n2 = Get l_assetPath() -> (l_assetPath:String)
   n3 = Append(A=@n2, B=/, C=@n1)
   > Return Node!(
    execute
    path:String=@n2
    name:String=@n1
    package_name:String=@n3
    class_name:String=Append!(A=@n3, B=., C=Get SaveAssetName(), D:String=_C) -> (ReturnValue)
   )
   Get l_packageName() -> (l_packageName:String)
  ]}
 LoadData {nodes [
   LoadData() {is=title K2Node_FunctionEntry}
   > SetEditorProperty(Object=Self-Reference(), PropertyName=hasLoaded, PropertyValue:bool=MakeLiteralBool(Value=true), ChangeNotifyMode=Never)
   > n1 = DoesAssetExist(AssetPath=GetPathInfoForSave() {is=title K2Node_CallFunction}.package_name)
   > Branch(Condition=@n1)
   > n2 = Execute Python Script(
    PythonScript="import unreal\r\nmy_class = unreal.load_class(None, in_class_name)\r\ncdo = unreal.get_default_object(my_class)"
    in_class_name=GetPathInfoForSave().class_name
   ) {inputs [in_class_name], outputs [cdo]}
   > n3 = GetEditorProperty(Object=@n2.cdo, PropertyName=Get PropertyName_SaveDataMap() -> (PropertyName_SaveDataMap=None))
   > n4 = Set l_map(l_map=@n3)
   n5 = Find(TargetMap=@n4, Key=GetActorIDForSave() -> (ReturnValue=None) {is=title K2Node_CallFunction})
   > Branch(Condition=@n5.ReturnValue)
   > SetEditorProperty(Object=Self-Reference(), PropertyName=MapToLoad, PropertyValue:Soft World Reference Structure=@n5)
   Set privateMapToLoad(privateMapToLoad:Soft World Reference Structure) -> (Output_Get:Soft World Reference Structure)
  ]}
 UserConstructionScript {nodes [
   Construction Script()
   > Clear and Invalidate Timer by Handle!(execute, WorldContextObject:Object Reference, Handle=Get constructTimerHandle() -> (constructTimerHandle:Timer Handle Structure)) -> (then)
   > n1 = Set Timer by Event(Delegate=Create Event(), Time=0.1, bLooping=false)
   > Set constructTimerHandle(constructTimerHandle:Timer Handle Structure=@n1) -> (Output_Get:Timer Handle Structure)
   > SetEditorProperty(Object=Self-Reference(), PropertyName=bIsSpatiallyLoaded, PropertyValue:bool=MakeLiteralBool(Value=false))
  ]}
]
variables [
 constructTimerHandle {type=TimerHandle}
 currentSpawnedActorSoftRef {type=BuildingLevelInstance, default=null}
 hasLoaded {type=boolean}
 MapToLoad {type=SoftWorldReference, default {worldAsset=null}}
 privateMapToLoad {type=SoftWorldReference, default {worldAsset=null}}
 PropertyName_SaveDataMap {type=string, default=MapToLoad}
 SaveAssetName {type=string, default=FiringRange_SaveData}
]
